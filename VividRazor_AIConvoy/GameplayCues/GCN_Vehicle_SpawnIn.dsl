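; Burst gameplay cue notify that adjusts the spawn-in effect for AI convoy vehicles.
; On burst, it picks a smaller effect scale when the target is the sport motorcycle escort and a larger one for any other target.
; It writes that scale to the first particle component's Niagara "Global Scale" variable and raises the component above its spawn location.
; The Niagara-only OnBurst variant is empty.
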(fn OnBurst(EitherVFX-preferthisversion)
  (params (MyTarget :type Actor%20Object%20Reference) (Parameters :type Gameplay%20Cue%20Parameters%20Structure%20%28by%20ref%29) (ParticleComponents :type Array%20of%20FXSystem%20Component%20Object%20References) (AudioComponents :type Array%20of%20Audio%20Component%20Object%20References) (BurstCameraShakeInstance :type Camera%20Shake%20Base%20Object%20Reference) (BurstDecalInstance :type Decal%20Actor%20Object%20Reference))
  (Utilities|Casting|CastToMotorcycle_Sport_Vehicle_Convoy_Escort MyTarget
    (:then
      (LocalVariables|SetScale 4.0))
    (:CastFailed
      (LocalVariables|SetScale 1e+01)))
  (bind _asniagara_particle_system_component (Utilities|Casting|CastToNiagaraComponent (Utilities|Array|Get(acopy) ParticleComponents)))
  (Niagara|SetNiagaraVariable(Float) _asniagara_particle_system_component "Global Scale" (LocalVariables|GetScale))
  (Transformation|SetWorldLocation _asniagara_particle_system_component (+ (Transformation|GetWorldLocation _asniagara_particle_system_component) "0.000000,0.000000,120.000000") false false)
  (return))

(fn OnBurst(Niagara)
  (params (MyTarget :type Actor%20Object%20Reference) (Parameters :type Gameplay%20Cue%20Parameters%20Structure%20%28by%20ref%29) (ParticleComponents :type Array%20of%20Niagara%20Particle%20System%20Component%20Object%20References) (AudioComponents :type Array%20of%20Audio%20Component%20Object%20References) (BurstCameraShakeInstance :type Camera%20Shake%20Base%20Object%20Reference) (BurstDecalInstance :type Decal%20Actor%20Object%20Reference)))
